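# Simplified parent material for the Octane Neon Light vehicle wheel, usable on skeletal meshes and built on material attributes.
# It starts from MF_WheelParent_Simple and alters only Base Color, where the red vertex color channel masks in a Fresnel-based brightening borrowed from the Cola Holographic wheel's function folder.
# The result then passes through the shared vehicle damage function and the DeRez transition function.
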Material /VehicleCosmetics/Wheels/OctaneNeonLight/Materials/M_WheelOctaneNeonLight_Parent_Simple
graph {
 properties {
  bAutomaticallySetUsageInEditor=false
  bUseMaterialAttributes=true
  bUsedWithMeshDeformer=true
  bUsedWithSkeletalMesh=true
  pixelDepthOffsetMode=PDOM_Legacy
  usedShadingModels=Default Lit
 }
 nodes [
  n1 = MaterialFunctionCall() {materialFunction=&MF_WheelParent_Simple}
  n2 = GetMaterialAttributes(None=@n1.MaterialAttributes) {attributeGetTypes [#1]}
  n3 = Multiply(A=@n2.Base Color, B=@n1.VertexColorR)
  n4 = Constant() {r=4}
  n5 = LinearInterpolate(
   A=@n2.Base Color
   B=LinearInterpolate(
    A=@n3
    B=Add(A=@n3) {constB=0.4}
    Alpha=MaterialFunctionCall!(Normal, Exponent=@n4) {materialFunction=/VehicleCosmetics/Wheels/Cola_Holographic/Materials/MF_Fresnel_Classic}.Result
   )
   Alpha=@n1.VertexColorR
  )
  MaterialFunctionCall(In=MaterialFunctionCall(Material=SetMaterialAttributes!(MaterialAttributes=@n2.MaterialAttributes, Base Color=@n5) {attributeSetTypes [#1]}) {materialFunction=&MF_BaseVehicle_Damage}.Result) {materialFunction=&MF_VehicleDeRez_Transition}
 ]
}
