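# Masked, two-sided material for the special-effect part of the Octane Neon Light vehicle wheel, drawing glowing neon in four tintable colors.
# Packed mask channels, some split into half and full levels by Step tests, provide regions for color assembly and opacity.
# A two-second Time loop produces staggered on/off thresholds, which appear to sequence the layered emissive colors.
# Fresnel terms darken base color and emission, and TwoSidedSign flips the normal on back faces.
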Material /VehicleCosmetics/Wheels/OctaneNeonLight/Materials/M_Wheel_OctaneNeonLight_SE
graph {
 properties {
  bAutomaticallySetUsageInEditor=false
  bUseMaterialAttributes=true
  bUsedWithMeshDeformer=true
  bUsedWithSkeletalMesh=true
  blendMode=BLEND_Masked
  twoSided=true
  usedShadingModels=Default Lit
 }
 nodes [
  n1 = Time()
  n2 = Clamp(None=Frac(None=Multiply(A=@n1) {constB=0.5}))
  NamedRerouteDeclaration(None=OneMinus(None=Step(Y=@n2) {constX=0.375})) {name=0.375, nodeColor=<rgba 0 0.25 1 1>}
  NamedRerouteDeclaration(None=OneMinus(None=Step(Y=@n2) {constX=0.565})) {name=0.565, nodeColor=<rgba 0 0.25 1 1>}
  NamedRerouteDeclaration(None=OneMinus(None=Step(Y=@n2) {constX=0.6375})) {name=0.6375, nodeColor=<rgba 0 0.25 1 1>}
  NamedRerouteDeclaration(None=OneMinus(None=Step(Y=@n2) {constX=0.71})) {name=0.71, nodeColor=<rgba 0 0.25 1 1>}
  NamedRerouteDeclaration(None=OneMinus(None=Step(Y=@n2) {constX=0.2})) {name=0.2, nodeColor=<rgba 0 0.25 1 1>}
  n3 = TextureSampleParameter2D() {
   group=00 Base
   parameterName=Masks
   samplerType=SAMPLERTYPE_LinearColor
   sortPriority=0
   texture=/VehicleCosmetics/Wheels/OctaneNeonLight/Textures/T_OctaneNeonLight_SE_M
  }
  NamedRerouteDeclaration(None=@n3.R) {name=Masks_R, nodeColor=<rgba 1 0 0.75 1>}
  n4 = NamedRerouteDeclaration(None=@n3.G) {name=Masks_G, nodeColor=<rgba 0 1 0 1>}
  n5 = Subtract(A=@n4) {constB=0.5}
  NamedRerouteDeclaration(None=Step(Y=Abs(None=@n5)) {constX=0.1}) {name=Masks_G_50, nodeColor=<rgba 0 1 0 1>}
  n6 = Subtract(A=@n4)
  NamedRerouteDeclaration(None=Step(Y=Abs(None=@n6)) {constX=0.45}) {name=Masks_G_100, nodeColor=<rgba 0 1 0 1>}
  n7 = NamedRerouteDeclaration(None=@n3.B) {name=Masks_B, nodeColor=<rgba 0 0 1 1>}
  n8 = Subtract(A=@n7) {constB=0.5}
  NamedRerouteDeclaration(None=Step(Y=Abs(None=@n8)) {constX=0.1}) {name=Masks_B_50, nodeColor=<rgba 0 0 1 1>}
  n9 = Subtract(A=@n7)
  NamedRerouteDeclaration(None=Step(Y=Abs(None=@n9)) {constX=0.45}) {name=Masks_B_100, nodeColor=<rgba 0 0 1 1>}
  NamedRerouteDeclaration(None=@n3.A) {name=Masks_A, nodeColor=<rgba 1 1 1 1>}
  n10 = Add(A=NamedRerouteUsage(), B=SmoothStep(Value=NamedRerouteUsage()) {constMax=0.5})
  n11 = NamedRerouteUsage()
  NamedRerouteDeclaration(None=Add(A=Add(A=@n10, B=@n11), B=NamedRerouteUsage())) {name=Opacity, nodeColor=<rgba 0 1 0.5312 1>}
  n12 = Multiply(A=VectorParameter() {group=00 Base, parameterName=Neon Color 1, defaultValue=<rgba 0 0.8492 1 1>}.RGB, B=NamedRerouteUsage())
  n13 = VectorParameter() {group=00 Base, parameterName=Neon Color 2, defaultValue=<rgba 0 1 0.0579 1>}
  n14 = NamedRerouteUsage()
  n15 = Multiply(A=@n13.RGB, B=@n14)
  n16 = NamedRerouteUsage()
  n17 = Multiply(A=@n13.RGB, B=@n16)
  n18 = VectorParameter() {group=00 Base, parameterName=Neon Color 3, defaultValue=<rgba 1 1 1 1>}
  n19 = NamedRerouteUsage()
  n20 = Multiply(A=@n18.RGB, B=@n19)
  n21 = Add(A=Add(A=Add(A=@n12, B=@n15), B=@n17), B=@n20)
  n22 = NamedRerouteUsage()
  n23 = Multiply(A=@n18.RGB, B=@n22)
  n24 = NamedRerouteUsage()
  n25 = Multiply(A=VectorParameter() {group=00 Base, parameterName=Neon Color 4, defaultValue=<rgba 1 1 0 1>}.RGB, B=@n24)
  n26 = LinearInterpolate(B=Add(A=Add(A=@n21, B=@n23), B=@n25) {desc=Assemble Colors for BC}, Alpha=NamedRerouteUsage()) {constA=1, desc=Gotten rid of black edges}
  n27 = NamedRerouteUsage()
  NamedRerouteDeclaration(None=LinearInterpolate(
    A=@n26
    B=Multiply(A=@n26) {constB=0.6}
    Alpha=MaterialFunctionCall!(Normal=@n27, Exponent) {materialFunction=/VehicleCosmetics/Wheels/Cola_Holographic/Materials/MF_Fresnel_Classic}.Result
   )) {name=BC, nodeColor=<rgba 0.5234 1 0 1>}
  n28 = NamedRerouteUsage()
  n29 = LinearInterpolate(A=LinearInterpolate(A=@n12, B=@n15, Alpha=Multiply(A=@n14, B=@n28)), B=@n17, Alpha=Multiply(A=@n16, B=NamedRerouteUsage()))
  n30 = Multiply(A=@n19, B=NamedRerouteUsage())
  n31 = NamedRerouteUsage()
  n32 = LinearInterpolate(
   A=LinearInterpolate(A=LinearInterpolate(A=@n29, B=@n20, Alpha=@n30), B=@n23, Alpha=Multiply(A=@n22, B=@n31))
   B=@n25
   Alpha=Multiply(A=@n24, B=NamedRerouteUsage())
  ) {desc=Emissive}
  n33 = Multiply(A=@n32) {constB=0.25}
  n34 = NamedRerouteUsage()
  n35 = Constant() {r=4}
  n36 = Clamp(None=ConstantBiasScale(None=MaterialFunctionCall(Normal=@n34, Exponent=@n35) {materialFunction=&MF_Fresnel_Classic}.Result) {bias=0, scale=5})
  NamedRerouteDeclaration(None=Multiply(
    A=LinearInterpolate(A=@n32, B=@n33, Alpha=@n36)
    B=ScalarParameter() {defaultValue=10, group=00 Base, parameterName=Emission Intensity, sortPriority=4}
   )) {name=Emissive, nodeColor=<rgba 1 0.5156 0 1>}
  n37 = TextureSampleParameter2D() {
   group=00 Base
   parameterName=Normal
   samplerType=SAMPLERTYPE_Normal
   sortPriority=1
   texture=/VehicleCosmetics/Wheels/OctaneNeonLight/Textures/T_OctaneNeonLight_SpecialEffect_N
  }
  NamedRerouteDeclaration(None=Multiply(A=@n37.RGB, B=TwoSidedSign())) {name=Normal, nodeColor=<rgba 0 0.3672 1 1>}
  n38 = NamedRerouteUsage()
  SetMaterialAttributes!(
   MaterialAttributes
   Base Color=NamedRerouteUsage()
   Normal=NamedRerouteUsage()
   Emissive Color=NamedRerouteUsage()
   Metallic=ScalarParameter() {group=00 Base, parameterName=Metallic, sortPriority=2}
   Roughness=ScalarParameter() {defaultValue=0.2, group=00 Base, parameterName=Roughness, sortPriority=3}
   Opacity=@n38
   Opacity Mask=@n38
  ) {attributeSetTypes [
    #1
    #2
    #3
    #4
    #5
    #6
    #7
   ]}
 ]
}
